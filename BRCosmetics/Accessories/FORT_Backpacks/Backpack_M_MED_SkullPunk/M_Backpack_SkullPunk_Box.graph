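# Translucent unlit effect material for the SkullPunk backpack box, flagged for Niagara mesh particles.
# A curve atlas row, sampled by a blend of UV V and transformed Z height remapped through dynamic parameters, mixes two colors with its red channel and scales them with its green.
# Screen-space scanlines from a panned horizontal coordinate, faded by a dynamic parameter, multiply that color.
# Particle color tints the result before it enters the shared emissive-adjustments function.
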Material /BRCosmetics/Accessories/FORT_Backpacks/Backpack_M_MED_SkullPunk/M_Backpack_SkullPunk_Box
graph {
 properties {
  bUsedWithNiagaraMeshParticles=true
  blendMode=BLEND_Translucent
  pixelDepthOffsetMode=PDOM_Legacy
  shadingModel=MSM_Unlit
  translucencyPass=MTP_BeforeDOF
  usedShadingModels=Unlit
 }
 nodes [
  n1 = TextureCoordinate()
  n2 = WorldPosition() {worldPositionShaderOffset=WPT_ExcludeAllShaderOffsets}
  n3 = DynamicParameter() {parameterIndex=3, defaultValue=<rgba 0 1 1 1>, paramNames [Scanline Lerp, Param2, Divide, Zheight]}
  n4 = Subtract(
   A=MaterialFunctionCall(Float3=TransformPosition(None=@n2.XYZ) {transformSourceType=TRANSFORMPOSSOURCE_World}) {materialFunction=&BreakOutFloat3Components}.B
   B=@n3.Divide
  )
  n5 = CurveAtlasRowParameter(CurveTime=LinearInterpolate(A=ComponentMask(None=@n1) {g=true}, B=Divide(A=@n4, B=@n3.Zheight) {constB=256}) {constAlpha=0.8}) {
   atlas=/Game/Athena/LinearColorCurves/VFX_CurveAtlas
   curve=/Game/Accessories/FORT_Backpacks/Backpack_M_MED_SkullPunk/Curve_Backpack_SkullPunk
   defaultValue=58
  }
  n6 = Multiply(
   A=LinearInterpolate(
    A=VectorParameter() {parameterName=ColorA, defaultValue=<rgba 0 1.5 1.2 1>}.RGB
    B=VectorParameter() {parameterName=ColorB, defaultValue=<rgba 0.025 0.0075 0.0162 1>}.RGB
    Alpha=@n5.R
   )
   B=@n5.G
  )
  n7 = Panner(Coordinate=ScreenPosition().ViewportUV, Speed=AppendVector(A=ScalarParameter() {group=Scanline, parameterName=Scanline Pan}, B=Constant())) {speedX=0.03}
  n8 = ScalarParameter() {defaultValue=450, group=Scanline, parameterName=Scanline Amount}
  n9 = Frac(None=ComponentMask(None=Multiply(A=@n7, B=@n8) {constB=200}) {r=true})
  n10 = Add(A=Multiply(A=@n9, B=OneMinus(None=@n9)), B=ScalarParameter() {defaultValue=-0.1, group=Scanline, parameterName=ScanlineSize})
  n11 = LinearInterpolate(
   B=LinearInterpolate(
    A=ScalarParameter() {defaultValue=0.7, group=Scanline, parameterName=Scanlines}
    B=ScalarParameter() {defaultValue=1, group=Scanline, parameterName=Scanlines Light}
    Alpha=Ceil(None=@n10)
   )
   Alpha=@n3.Scanline Lerp
  ) {constA=1}
  MaterialFunctionCall(Emissive=Multiply(A=Multiply(A=@n6, B=@n11), B=ParticleColor().RGB)) {materialFunction=&MF_FX_EmissiveAdjustments}
 ]
}
