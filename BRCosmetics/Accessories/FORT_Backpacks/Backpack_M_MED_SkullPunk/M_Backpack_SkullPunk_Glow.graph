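# Unlit, additive, two-sided glow material for the SkullPunk backpack cosmetic, usable on Niagara sprites.
# Emissive blends two color parameters using a curve-atlas row sampled along V, tinted by particle color and passed through an emissive adjustment function.
# Opacity multiplies a glow texture by a V-edge falloff, view-angle fade and particle alpha.
# A shading path switch applies depth fade on default and deferred paths, skipping it on forward and mobile.
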Material /BRCosmetics/Accessories/FORT_Backpacks/Backpack_M_MED_SkullPunk/M_Backpack_SkullPunk_Glow
graph {
 properties {
  bUsedWithNiagaraSprites=true
  blendMode=BLEND_Additive
  pixelDepthOffsetMode=PDOM_Legacy
  shadingModel=MSM_Unlit
  twoSided=true
  usedShadingModels=Unlit
 }
 nodes [
  n1 = VectorParameter() {parameterName=ColorA, defaultValue=<rgba 0 0.35 0.35 1>}
  n2 = VectorParameter() {parameterName=ColorB, defaultValue=<rgba 0.28 0 0.1867 1>}
  n3 = CurveAtlasRowParameter(CurveTime=ComponentMask(None=TextureCoordinate()) {g=true}) {
   atlas=/Game/Athena/LinearColorCurves/VFX_CurveAtlas
   curve=/Game/Accessories/FORT_Backpacks/Backpack_M_MED_SkullPunk/Curve_Backpack_SkullPunk
   defaultValue=58
  }
  n4 = ParticleColor()
  MaterialFunctionCall(Emissive=Multiply(A=LinearInterpolate(A=@n1.RGB, B=@n2.RGB, Alpha=@n3.B), B=@n4.RGB)) {materialFunction=&MF_FX_EmissiveAdjustments}
  n5 = TextureCoordinate()
  n6 = ComponentMask(None=@n5) {g=true}
  n7 = OneMinus(None=ComponentMask(None=@n5) {g=true})
  n8 = Multiply(A=Power(Base=Multiply(A=@n6, B=@n7)) {constExponent=6}) {constB=2000}
  n9 = Multiply(
   A=Multiply(
    A=Saturate(None=@n8)
    B=TextureSample() {texture=&/T_Backpack_SkullPunk_Glow}.G
   )
   B=MaterialFunctionCall("FadeSensitivity (optional)"=ScalarParameter() {defaultValue=1.5, group=Digital_glow, parameterName=ViewAngleFade}) {materialFunction=&FadeBasedOnViewAngle}.Result
  )
  n10 = DepthFade(Opacity=@n9, FadeDistance=ScalarParameter() {defaultValue=10, parameterName=Depth_fade})
  Multiply(A=@n4.A, B=ShadingPathSwitch(Default=@n10, Deferred=@n10, Forward=@n9, Mobile=@n9))
 ]
}
