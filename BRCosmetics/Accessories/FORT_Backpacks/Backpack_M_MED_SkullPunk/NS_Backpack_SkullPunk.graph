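# Niagara effect for the SkullPunk backpack cosmetic: an audio-reactive, glitchy bar display with glow and lights.
# Top and bottom rows of curve-shaped mesh bars undulate with noise, while camera angle drives a material parameter.
# Event audio amplitude boosts bar motion, brightness, glow and two tinted lights, with front-end and time-of-day adjustments on bars and glow.
# Looping cyan/magenta sprites jitter as glitches, and User.Offset Back shifts every emitter along Y.
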NiagaraSystem /BRCosmetics/Accessories/FORT_Backpacks/Backpack_M_MED_SkullPunk/NS_Backpack_SkullPunk
dependencies {
 usedDataInterfaces [
  &NiagaraDataInterfaceCurve
  &NiagaraDataInterfaceCamera
  &NiagaraDataInterfaceColorCurve
  &NiagaraDataInterfaceVector2DCurve
  &NiagaraDataInterfaceArrayColor
  &NiagaraDataInterfaceArrayFloat3
 ]
 usedDynamicInputs []
 usedModules [&SystemState]
}
emitters {
 Box_Bottom {
  emitterUpdateScript [
   &EmitterState(Life Cycle Mode=::NewEnumerator1, Loop Behavior=::NewEnumerator1, Loop Duration Mode=::NewEnumerator1, UseLoopDelay=false)
   &SpawnBurst_Instantaneous(Spawn Count=40, Use Spawn Probability=false, UseLoopCountLimit=false)
  ]
  particleSpawnScript [
   &FN_SetSpawnGroup(Normalize Spawn Group=true, Particles Per Group=40, Pattern=::NewEnumerator0)
   &InitializeParticle(
    Color=<rgba 1.2 1.2 1.2 1>
    Mesh Scale=&MakeVector(
     X=&FN_FrontEndFloatModifier(Gameplay=0.075, Front End=0.075)
     Y=0.1
     Z=&FloatFromCurve(
      FloatCurve=DataInterface() {propertyValues {Curve {keys [
          {interpMode=RCIM_Cubic, leaveTangent=2.2735, tangentMode=RCTM_Break, time=0, value=0}
          {arriveTangent=3.1375, interpMode=RCIM_Cubic, leaveTangent=-2.9395, tangentMode=RCTM_Break, time=0.5, value=1}
          {arriveTangent=-2.5624, interpMode=RCIM_Cubic, tangentMode=RCTM_Break, time=1, value=0}
         ]}}}
      CurveIndex=$Particles.Normalized Spawn Group
     )
    )
    Mesh Scale Mode=::NewEnumerator4
    Position Offset=&MakeVector(
     X=&Lerp_Float(A=-20, B=20, Alpha=$Particles.Normalized Spawn Group)
     Y=&Add_Float(
      A=$User.Offset Back
      B=&Multiply_Float(
       A=-1
       B=&FloatFromCurve(
        FloatCurve=DataInterface() {propertyValues {Curve {keys [
            {arriveTangent=6.4071, interpMode=RCIM_Cubic, leaveTangent=6.4071, tangentMode=RCTM_User, time=0, value=0}
            {interpMode=RCIM_Cubic, tangentMode=RCTM_User, time=0.5, value=3}
            {arriveTangent=-5.1876, interpMode=RCIM_Cubic, leaveTangent=-5.1876, tangentMode=RCTM_User, time=1, value=0}
           ]}}}
        CurveIndex=$Particles.Normalized Spawn Group
       )
      )
     )
     Z=&FloatFromCurve(
      FloatCurve=DataInterface() {propertyValues {Curve {keys [
          {interpMode=RCIM_Cubic, tangentMode=RCTM_Break, time=0.1, value=0}
          {interpMode=RCIM_Cubic, tangentMode=RCTM_Break, time=0.5, value=-6}
          {interpMode=RCIM_Cubic, tangentMode=RCTM_Break, time=0.9, value=0}
         ]}}}
      CurveIndex=$Particles.Normalized Spawn Group
     )
    )
    UsePositionOffset=true
    Write Material Random=true
   )
  ]
  particleUpdateScript [
   &ParticleState(Kill Particles When Lifetime Has Elapsed=false, Let Infinitely Lived Particles Die When Emitter Deactivates=true)
   /SetVariables(Particles.Dot:NiagaraFloat=&DotProduct(
     Vector A=&ReturnCameraVector(Camera Property=::NewEnumerator4, Coordinate Space=::Local, Origin Position=$Particles.Position)
     Vector B=<0 1 0>
    )) {isSetParameters=true, module=SetVariables_#1}
   &ScaleMeshSize(Scale Factor=&MakeVector(
     X=&Lerp_Float(
      A=1
      B=1
      Alpha=&FloatFromCurve(
       FloatCurve=DataInterface() {propertyValues {Curve {keys [{interpMode=RCIM_Cubic, tangentMode=RCTM_User, time=0, value=0}, {interpMode=RCIM_Cubic, tangentMode=RCTM_User, time=1, value=1}]}}}
       CurveIndex=&Abs_Float(Float:NiagaraFloat=$Particles.Dot)
      )
     )
     Y=1
     Z=&Add_Float(
      A=0.5
      B=&Abs_Float(Float:NiagaraFloat=&self:NoiseFloat(
        Intensity:NiagaraFloat=&FN_FrontEndFloatModifier(
         Gameplay=&Lerp_Float(A=0.1, B=1.2, Alpha=$NPC.FortniteNPC.EventAudio_Amplitude)
         Front End=&Lerp_Float(A=0.1, B=0.3, Alpha=$NPC.FortniteNPC.EventAudio_Amplitude)
        )
        Position:NiagaraFloat=&Lerp_Float(A=0, B=10, Alpha=$Particles.Normalized Spawn Group)
        Speed:NiagaraFloat=1.7
       ))
     )
    ))
   /SetVariables(Particles.Position=&Add_Vector(
     A=Linked() {linkedVariable {name=Particles.Initial.Position, type=NiagaraPosition}}
     B=&MakeVector(
      X=0
      Y=&Multiply_Float(
       A=&self:NoiseFloat(
        Intensity:NiagaraFloat=0.5
        Position:NiagaraFloat=&Lerp_Float(A=0, B=1.5, Alpha=$Particles.Normalized Spawn Group)
        Speed:NiagaraFloat=1
       )
       B=&FloatFromCurve(
        FloatCurve=DataInterface() {propertyValues {Curve {keys [
            {interpMode=RCIM_Cubic, leaveTangent=10.8964, tangentMode=RCTM_Break, time=0, value=-0.5}
            {arriveTangent=0.0731, interpMode=RCIM_Cubic, leaveTangent=0.1354, tangentMode=RCTM_Break, time=0.5, value=1.5}
            {arriveTangent=-10.8124, interpMode=RCIM_Cubic, tangentMode=RCTM_Break, time=1, value=-0.5}
           ]}}}
        CurveIndex=$Particles.Normalized Spawn Group
        Scale Curve=-1
       )
      )
      Z=&Multiply_Float(
       A=&self:NoiseFloat(
        Intensity:NiagaraFloat=0.75
        Position:NiagaraFloat=&Lerp_Float(A=0, B=8, Alpha=$Particles.Normalized Spawn Group)
        Speed:NiagaraFloat=1.5
       )
       B=&FloatFromCurve(
        FloatCurve=DataInterface() {propertyValues {Curve {keys [
            {interpMode=RCIM_Cubic, leaveTangent=0.0147, tangentMode=RCTM_Break, time=0, value=0}
            {arriveTangent=0.0731, interpMode=RCIM_Cubic, leaveTangent=0.1354, tangentMode=RCTM_Break, time=0.5, value=1.5}
            {arriveTangent=0.1088, interpMode=RCIM_Cubic, tangentMode=RCTM_Break, time=1, value=0}
           ]}}}
        CurveIndex=$Particles.Normalized Spawn Group
       )
      )
     )
    )) {isSetParameters=true, module=SetVariables_#2}
   &DynamicMaterialParameters(
    Index 3 Param 1=&Lerp_Float(
     A=1
     B=0
     Alpha=&FloatFromCurve(
      FloatCurve=DataInterface() {propertyValues {Curve {keys [
          {arriveTangent=-0.0212, interpMode=RCIM_Cubic, leaveTangent=-0.0212, tangentMode=RCTM_User, time=0, value=0}
          {arriveTangent=0.3102, interpMode=RCIM_Cubic, leaveTangent=0.3102, tangentMode=RCTM_User, time=0.9, value=0.2}
          {arriveTangent=0.0069, interpMode=RCIM_Cubic, leaveTangent=0.0069, tangentMode=RCTM_User, time=0.97, value=1}
         ]}}}
      CurveIndex=&Abs_Float(Float:NiagaraFloat=$Particles.Dot)
     )
    )
    Index 3 Param 3=5
    Index 3 Param 4=-17
    Module.Param0WriteEnabled3=true
    Module.Param1WriteEnabled3=false
    Module.Param2WriteEnabled3=true
    Module.Param3WriteEnabled3=true
    Write Parameter Index 0=false
    Write Parameter Index 3=true
   )
   &ScaleColor(
    Scale RGB=&Multiply_VectorByFloat(
     Vector=&VectorFromFloat(Value=&FN_FrontEndFloatModifier(
       Gameplay=&Lerp_Float(
        A=1
        B=15
        Alpha=&FloatFromCurve(FloatCurve=DataInterface() {propertyValues {Curve {keys [<at 0.75 0>,<at 1 1>]}}}, CurveIndex=$NPC.FortniteNPC.EventAudio_Amplitude)
       )
       Front End=2
      ))
     Float=&FNTimeOfDayScalar(Night Scale=0.5, Morning Scale=0.1, Day Scale=0.8, Evening Scale=0.2)
    )
    ScaleA=true
    ScaleRGB=true
   )
  ]
  rendererClasses [&NiagaraMeshRendererProperties]
  renderers [{rendererClass=&NiagaraMeshRendererProperties, rendererIndex=0}]
 }
 Box_Top {
  emitterUpdateScript [
   &EmitterState(Life Cycle Mode=::NewEnumerator1, Loop Behavior=::NewEnumerator1, Loop Duration Mode=::NewEnumerator1, UseLoopDelay=false)
   &SpawnBurst_Instantaneous(Spawn Count=40, Use Spawn Probability=false, UseLoopCountLimit=false)
  ]
  particleSpawnScript [
   &FN_SetSpawnGroup(Normalize Spawn Group=true, Particles Per Group=40, Pattern=::NewEnumerator0)
   &InitializeParticle(
    Color=<rgba 1.5 1.5 1.5 1>
    Mesh Scale=&MakeVector(
     X=&FN_FrontEndFloatModifier(Gameplay=0.075, Front End=0.075)
     Y=0.1
     Z=&FloatFromCurve(
      FloatCurve=DataInterface() {propertyValues {Curve {keys [
          {interpMode=RCIM_Cubic, tangentMode=RCTM_Break, time=0, value=0}
          {arriveTangent=3.1375, interpMode=RCIM_Cubic, leaveTangent=-2.9395, tangentMode=RCTM_Break, time=0.5, value=1}
          {interpMode=RCIM_Cubic, tangentMode=RCTM_Break, time=1, value=0}
         ]}}}
      CurveIndex=$Particles.Normalized Spawn Group
     )
    )
    Mesh Scale Mode=::NewEnumerator4
    Position Offset=&MakeVector(
     X=&Lerp_Float(A=-20, B=20, Alpha=$Particles.Normalized Spawn Group)
     Y=&Add_Float(
      A=$User.Offset Back
      B=&Multiply_Float(
       A=-1
       B=&FloatFromCurve(
        FloatCurve=DataInterface() {propertyValues {Curve {keys [
            {arriveTangent=2.8486, interpMode=RCIM_Cubic, leaveTangent=2.8486, tangentMode=RCTM_User, time=0, value=0}
            {interpMode=RCIM_Cubic, tangentMode=RCTM_User, time=0.5, value=3}
            {arriveTangent=-2.2098, interpMode=RCIM_Cubic, leaveTangent=-2.2098, tangentMode=RCTM_User, time=1, value=0}
           ]}}}
        CurveIndex=$Particles.Normalized Spawn Group
       )
      )
     )
     Z=0
    )
    UsePositionOffset=true
    Write Material Random=true
   )
  ]
  particleUpdateScript [
   &ParticleState(Kill Particles When Lifetime Has Elapsed=false, Let Infinitely Lived Particles Die When Emitter Deactivates=true)
   /SetVariables(Particles.Dot:NiagaraFloat=&DotProduct(
     Vector A=&ReturnCameraVector(Camera Property=::NewEnumerator4, Coordinate Space=::Local, Origin Position=$Particles.Position)
     Vector B=<0 1 0>
    )) {isSetParameters=true, module=SetVariables_#1}
   &ScaleMeshSize(Scale Factor=&MakeVector(
     X=&Lerp_Float(
      A=1
      B=1
      Alpha=&FloatFromCurve(
       FloatCurve=DataInterface() {propertyValues {Curve {keys [{interpMode=RCIM_Cubic, tangentMode=RCTM_User, time=0, value=0}, {interpMode=RCIM_Cubic, tangentMode=RCTM_User, time=1, value=1}]}}}
       CurveIndex=&Abs_Float(Float:NiagaraFloat=$Particles.Dot)
      )
     )
     Y=1
     Z=&Add_Float(
      A=1.5
      B=&Abs_Float(Float:NiagaraFloat=&self:NoiseFloat(
        Intensity:NiagaraFloat=&FN_FrontEndFloatModifier(
         Gameplay=&Lerp_Float(A=0.3, B=1.1, Alpha=$NPC.FortniteNPC.EventAudio_Amplitude)
         Front End=&Lerp_Float(A=0.3, B=0.5, Alpha=$NPC.FortniteNPC.EventAudio_Amplitude)
        )
        Position:NiagaraFloat=&Lerp_Float(A=0, B=6, Alpha=$Particles.Normalized Spawn Group)
        Speed:NiagaraFloat=1.7
       ))
     )
    ))
   /SetVariables(Particles.Position=&Add_Vector(
     A=Linked() {linkedVariable {name=Particles.Initial.Position, type=NiagaraPosition}}
     B=&MakeVector(
      X=0
      Y=&Multiply_Float(
       A=&self:NoiseFloat(
        Intensity:NiagaraFloat=0.4
        Position:NiagaraFloat=&Lerp_Float(A=0, B=2, Alpha=$Particles.Normalized Spawn Group)
        Speed:NiagaraFloat=0.75
       )
       B=&FloatFromCurve(
        FloatCurve=DataInterface() {propertyValues {Curve {keys [
            {interpMode=RCIM_Cubic, leaveTangent=16.487, tangentMode=RCTM_Break, time=0, value=-1}
            {arriveTangent=0.0731, interpMode=RCIM_Cubic, leaveTangent=0.1354, tangentMode=RCTM_Break, time=0.5, value=2}
            {arriveTangent=-21.3534, interpMode=RCIM_Cubic, tangentMode=RCTM_Break, time=1, value=-1}
           ]}}}
        CurveIndex=$Particles.Normalized Spawn Group
        Scale Curve=0.5
       )
      )
      Z=&Multiply_Float(
       A=&self:NoiseFloat(
        Intensity:NiagaraFloat=&Lerp_Float(A=1, B=2, Alpha=$NPC.FortniteNPC.EventAudio_Amplitude)
        Position:NiagaraFloat=&Lerp_Float(A=0, B=8, Alpha=$Particles.Normalized Spawn Group)
        Speed:NiagaraFloat=1.5
       )
       B=&FloatFromCurve(
        FloatCurve=DataInterface() {propertyValues {Curve {keys [
            {interpMode=RCIM_Cubic, leaveTangent=0.0147, tangentMode=RCTM_Break, time=0, value=0}
            {arriveTangent=0.0731, interpMode=RCIM_Cubic, leaveTangent=0.1354, tangentMode=RCTM_Break, time=0.5, value=1.5}
            {arriveTangent=0.1088, interpMode=RCIM_Cubic, tangentMode=RCTM_Break, time=1, value=0}
           ]}}}
        CurveIndex=$Particles.Normalized Spawn Group
       )
      )
     )
    )) {isSetParameters=true, module=SetVariables_#2}
   &DynamicMaterialParameters(
    Index 3 Param 1=&Lerp_Float(
     A=1
     B=0
     Alpha=&FloatFromCurve(
      FloatCurve=DataInterface() {propertyValues {Curve {keys [
          {arriveTangent=-0.0212, interpMode=RCIM_Cubic, leaveTangent=-0.0212, tangentMode=RCTM_User, time=0, value=0}
          {arriveTangent=0.3102, interpMode=RCIM_Cubic, leaveTangent=0.3102, tangentMode=RCTM_User, time=0.94, value=0.2}
          {arriveTangent=0.0069, interpMode=RCIM_Cubic, leaveTangent=0.0069, tangentMode=RCTM_User, time=0.97, value=1}
         ]}}}
      CurveIndex=&Abs_Float(Float:NiagaraFloat=$Particles.Dot)
     )
    )
    Index 3 Param 3=15
    Index 3 Param 4=-25
    Module.Param0WriteEnabled3=true
    Module.Param1WriteEnabled3=false
    Module.Param2WriteEnabled3=true
    Module.Param3WriteEnabled3=true
    Write Parameter Index 0=false
    Write Parameter Index 3=true
   )
   &ScaleColor(
    Scale RGB=&Multiply_VectorByFloat(
     Vector=&VectorFromFloat(Value=&FN_FrontEndFloatModifier(
       Gameplay=&Lerp_Float(
        A=1
        B=15
        Alpha=&FloatFromCurve(FloatCurve=DataInterface() {propertyValues {Curve {keys [<at 0.75 0>,<at 1 1>]}}}, CurveIndex=$NPC.FortniteNPC.EventAudio_Amplitude)
       )
       Front End=2
      ))
     Float=&FNTimeOfDayScalar(Night Scale=0.5, Morning Scale=0.1, Day Scale=0.8, Evening Scale=0.2)
    )
    ScaleA=true
    ScaleRGB=true
   )
  ]
  rendererClasses [&NiagaraMeshRendererProperties]
  renderers [{rendererClass=&NiagaraMeshRendererProperties, rendererIndex=0}]
 }
 Glitch_Bottom {
  emitterUpdateScript [
   &EmitterState(Life Cycle Mode=::NewEnumerator1, Loop Duration=5.6397, UseLoopDelay=false)
   &SpawnBurst_Instantaneous(Spawn Count=2, Use Spawn Probability=false, UseLoopCountLimit=false)
  ]
  particleSpawnScript [
   &FN_SetSpawnGroup(Normalize Spawn Group=true, Particles Per Group=7, Pattern=::NewEnumerator0)
   &InitializeParticle(
    Color=&LerpLinearColors(
     StartColor=<rgba 0 1.2 1.2 1>
     EndColor=<rgba 0.4 0.06 0.1733 1>
     LerpFactor=&UniformAOrBFloat(A=0, Distribution Weight=0.65)
    )
    Lifetime Max=2.5
    Lifetime Min=0.5
    Lifetime Mode=::NewEnumerator1
    Position Offset=&MakeVector(
     X=&FloatFromCurve(
      FloatCurve=DataInterface() {propertyValues {Curve {keys [
          {arriveTangent=-110.8723, interpMode=RCIM_Cubic, leaveTangent=-110.8723, tangentMode=RCTM_User, time=0, value=-15}
          {arriveTangent=47.6037, interpMode=RCIM_Cubic, leaveTangent=47.6037, tangentMode=RCTM_User, time=0.5, value=0}
          {arriveTangent=-57.341, interpMode=RCIM_Cubic, leaveTangent=-57.341, tangentMode=RCTM_User, time=1, value=15}
         ]}}}
      CurveIndex=$Particles.Normalized Spawn Group
     )
     Y=&Add_Float(A=$User.Offset Back, B=-3)
     Z=&FloatFromCurve(
      FloatCurve=DataInterface() {propertyValues {Curve {keys [
          {interpMode=RCIM_Cubic, tangentMode=RCTM_User, time=0, value=0}
          {arriveTangent=61.4469, interpMode=RCIM_Cubic, leaveTangent=-66.9619, tangentMode=RCTM_Break, time=0.5, value=15}
          {interpMode=RCIM_Cubic, tangentMode=RCTM_User, time=1, value=0}
         ]}}}
      CurveIndex=$Particles.Normalized Spawn Group
      Scale Curve=-1
     )
    )
    Sprite Size=&Multiply_Vector2D_ByFloat(Float=0.5, Vector2D=&RandomRangeVector2D(Minimum=<6.5 1.3>, Maximum=<8 1.7>))
    Sprite Size Mode=::NewEnumerator4
    UsePositionOffset=true
    Write Material Random=true
   )
   &DynamicMaterialParameters(
    Index 0 Param 1=&RandomRangeFloat(Minimum=0.05, Maximum=0.05)
    Index 0 Param 2=&RandomRangeFloat(Minimum=0.5, Maximum=1.5)
    Index 0 Param 3=&RandomRangeFloat(Minimum=0.2, Maximum=0.5)
    Index 0 Param 4=1
    Index 1 Param 1=10
    Module.Param0WriteEnabled=true
    Module.Param0WriteEnabled1=true
    Module.Param1WriteEnabled=true
    Module.Param1WriteEnabled1=true
    Module.Param2WriteEnabled=true
    Module.Param2WriteEnabled1=true
    Module.Param3WriteEnabled=true
    Module.Param3WriteEnabled1=true
    Write Parameter Index 1=true
   )
  ]
  particleUpdateScript [
   &ParticleState()
   &SolveForcesAndVelocity(Clamp Velocity=false, Limit Acceleration=false)
   &ScaleSpriteSize(Scale Factor=&Vector2DFromCurve(Vector2Curve=DataInterface() {propertyValues {XCurve {keys [<at 0 1>,<at 1 1>]}, YCurve {keys [<at 0 1>,<at 1 0.5>]}}}, ScaleCurve=<1 1>))
   /SetVariables(Particles.Position=&Add_Vector(
     A=Linked() {linkedVariable {name=Particles.Initial.Position, type=NiagaraPosition}}
     B=&RandomRangeVector(Minimum=<1 0 1>, Maximum=<-1 0 -1>)
    )) {isSetParameters=true, module=SetVariables_#3}
   &SpriteFacingAndAlignment(Sprite Alignment=<0 0 1>, Sprite Facing=<0 -1 0>, WriteAlignment=true, WriteFacing=true)
   &CameraOffset(Camera Offset Amount=5)
  ]
  rendererClasses [&NiagaraSpriteRendererProperties]
  renderers [
   {rendererClass=&NiagaraSpriteRendererProperties}
   {rendererClass=&NiagaraSpriteRendererProperties, rendererIndex=1}
  ]
 }
 Glitch_Top {
  emitterUpdateScript [
   &EmitterState(Life Cycle Mode=::NewEnumerator1, Loop Duration=1, UseLoopDelay=false)
   &SpawnBurst_Instantaneous(Spawn Count=2, Use Spawn Probability=false, UseLoopCountLimit=false)
  ]
  particleSpawnScript [
   &FN_SetSpawnGroup(Normalize Spawn Group=true, Particles Per Group=7, Pattern=::NewEnumerator0)
   &InitializeParticle(
    Color=&LerpLinearColors(
     StartColor=<rgba 0 1.2 1.2 1>
     EndColor=<rgba 0.4 0.06 0.1733 1>
     LerpFactor=&UniformAOrBFloat(A=0, Distribution Weight=0.45)
    )
    Lifetime Max=2.5
    Lifetime Min=0.5
    Lifetime Mode=::NewEnumerator1
    Position Offset=&MakeVector(
     X=&FloatFromCurve(
      FloatCurve=DataInterface() {propertyValues {Curve {keys [
          {arriveTangent=-110.8723, interpMode=RCIM_Cubic, leaveTangent=-110.8723, tangentMode=RCTM_User, time=0, value=-15}
          {arriveTangent=47.6037, interpMode=RCIM_Cubic, leaveTangent=47.6037, tangentMode=RCTM_User, time=0.5, value=0}
          {arriveTangent=-57.341, interpMode=RCIM_Cubic, leaveTangent=-57.341, tangentMode=RCTM_User, time=1, value=15}
         ]}}}
      CurveIndex=$Particles.Normalized Spawn Group
     )
     Y=&Add_Float(A=$User.Offset Back, B=-3)
     Z=&FloatFromCurve(
      FloatCurve=DataInterface() {propertyValues {Curve {keys [
          {interpMode=RCIM_Cubic, tangentMode=RCTM_User, time=0, value=0}
          {arriveTangent=61.4469, interpMode=RCIM_Cubic, leaveTangent=-66.9619, tangentMode=RCTM_Break, time=0.5, value=10}
          {interpMode=RCIM_Cubic, tangentMode=RCTM_User, time=1, value=0}
         ]}}}
      CurveIndex=$Particles.Normalized Spawn Group
      Scale Curve=1.4
     )
    )
    Sprite Size=&Multiply_Vector2D_ByFloat(Float=0.5, Vector2D=&RandomRangeVector2D(Minimum=<6.5 1.3>, Maximum=<8 1.7>))
    Sprite Size Mode=::NewEnumerator4
    UsePositionOffset=true
    Write Material Random=true
   )
   &DynamicMaterialParameters(
    Index 0 Param 1=&RandomRangeFloat(Minimum=0.05, Maximum=0.05)
    Index 0 Param 2=&RandomRangeFloat(Minimum=0.5, Maximum=1.5)
    Index 0 Param 3=&RandomRangeFloat(Minimum=0.2, Maximum=0.5)
    Index 0 Param 4=1
    Index 1 Param 1=1
    Module.Param0WriteEnabled=true
    Module.Param0WriteEnabled1=true
    Module.Param1WriteEnabled=true
    Module.Param1WriteEnabled1=true
    Module.Param2WriteEnabled=true
    Module.Param2WriteEnabled1=true
    Module.Param3WriteEnabled=true
    Module.Param3WriteEnabled1=true
    Write Parameter Index 1=true
   )
  ]
  particleUpdateScript [
   &ParticleState()
   &SolveForcesAndVelocity(Clamp Velocity=false, Limit Acceleration=false)
   &ScaleSpriteSize(Scale Factor=&Vector2DFromCurve(Vector2Curve=DataInterface() {propertyValues {XCurve {keys [<at 0 1>,<at 1 1>]}, YCurve {keys [<at 0 1>,<at 1 0.5>]}}}, ScaleCurve=<1 1>))
   /SetVariables(Particles.Position=&Add_Vector(
     A=Linked() {linkedVariable {name=Particles.Initial.Position, type=NiagaraPosition}}
     B=&RandomRangeVector(Minimum=<1 0 1>, Maximum=<-1 0 -1>)
    )) {isSetParameters=true, module=SetVariables_#3}
   &SpriteFacingAndAlignment(Sprite Alignment=<0 0 1>, Sprite Facing=<0 -1 0>, WriteAlignment=true, WriteFacing=true)
   &CameraOffset(Camera Offset Amount=5)
  ]
  rendererClasses [&NiagaraSpriteRendererProperties]
  renderers [
   {rendererClass=&NiagaraSpriteRendererProperties}
   {rendererClass=&NiagaraSpriteRendererProperties, rendererIndex=1}
  ]
 }
 Glow {
  emitterUpdateScript [
   &EmitterState(Life Cycle Mode=::NewEnumerator1, Loop Behavior=::NewEnumerator1, Loop Duration Mode=::NewEnumerator1, UseLoopDelay=false)
   &SpawnBurst_Instantaneous(Spawn Count=2, Use Spawn Probability=false, UseLoopCountLimit=false)
  ]
  particleSpawnScript [
   &FN_SetSpawnGroup(Number of Spawn Groups=2)
   &InitializeParticle(
    Color=<rgba 1 1 1 0.35>
    Position Offset=&Add_Vector(
     A=&MakeVector(X=0, Y=$User.Offset Back, Z=0)
     B=&SelectVectorFromArray(Direct Array Index=$Particles.SpawnGroup, Vector Selection Array=DataInterface() {propertyValues {FloatData [<0 -7 0>, <0 -2 0>]}})
    )
    Sprite Size=&Multiply_Vector2D_ByFloat(Float=1.2, Vector2D=<80 40>)
    Sprite Size Mode=::NewEnumerator4
    UsePositionOffset=true
    Write Material Random=true
   )
  ]
  particleUpdateScript [
   &ParticleState(Kill Particles When Lifetime Has Elapsed=false, Let Infinitely Lived Particles Die When Emitter Deactivates=true)
   &SolveForcesAndVelocity(Clamp Velocity=false, Limit Acceleration=false)
   &SpriteFacingAndAlignment(Sprite Alignment=<0 0 1>, Sprite Facing=<0 1 0>, WriteAlignment=true, WriteFacing=true)
   &ScaleColor(
    Scale Alpha=&FNTimeOfDayScalar(Night Scale=0.8, Morning Scale=0.3, Evening Scale=0.2)
    Scale RGB=&VectorFromFloat(Value=&Multiply_Float(
      A=&FN_FrontEndFloatModifier(Gameplay=1, Front End=3)
      B=&Lerp_Float(
       A=1
       B=1.5
       Alpha=&FloatFromCurve(
        FloatCurve=DataInterface() {propertyValues {Curve {keys [{interpMode=RCIM_Cubic, tangentMode=RCTM_User, time=0.75, value=0}, {interpMode=RCIM_Cubic, tangentMode=RCTM_User, time=1, value=1}]}}}
        CurveIndex=$NPC.FortniteNPC.EventAudio_Amplitude
       )
      )
     ))
    ScaleA=true
    ScaleRGB=true
   )
   &CameraOffset(Camera Offset Amount=5)
  ]
  rendererClasses [&NiagaraSpriteRendererProperties]
  renderers [{rendererClass=&NiagaraSpriteRendererProperties}]
 }
 Light {
  emitterUpdateScript [
   &EmitterState(Life Cycle Mode=::NewEnumerator1, Loop Behavior=::NewEnumerator1, Loop Duration Mode=::NewEnumerator1, UseLoopDelay=false)
   &SpawnBurst_Instantaneous(Spawn Count=2, Use Spawn Probability=false, UseLoopCountLimit=false)
  ]
  particleSpawnScript [
   &FN_SetSpawnGroup(Number of Spawn Groups=2)
   &InitializeParticle(
    Color=&SelectLinearColorFromArray(
     Array Sampling Mode=::NewEnumerator1
     Direct Array Index=$Particles.SpawnGroup
     Color Selection Array=DataInterface() {propertyValues {ColorData [<rgba 0.9 3 2.65 1>, <rgba 3 0.9 2.3 1>]}}
     Random Seed=?
    )
    Position Offset=&Add_Vector(
     A=&MakeVector(X=0, Y=$User.Offset Back, Z=0)
     B=&SelectVectorFromArray(Direct Array Index=$Particles.SpawnGroup, Vector Selection Array=DataInterface() {propertyValues {FloatData [<0 -7 7>, <0 -7 -5>]}})
    )
    Sprite Size Mode=::NewEnumerator3
    Uniform Sprite Size=2
    UsePositionOffset=true
    Write Material Random=true
   )
  ]
  particleUpdateScript [
   &ParticleState(Kill Particles When Lifetime Has Elapsed=false, Let Infinitely Lived Particles Die When Emitter Deactivates=true)
   &ScaleColor(
    Scale RGB=&VectorFromFloat(Value=&Lerp_Float(
      A=1
      B=5
      Alpha=&FloatFromCurve(
       FloatCurve=DataInterface() {propertyValues {Curve {keys [{interpMode=RCIM_Cubic, tangentMode=RCTM_User, time=0.5, value=0}, {interpMode=RCIM_Cubic, tangentMode=RCTM_User, time=1, value=1}]}}}
       CurveIndex=$NPC.FortniteNPC.EventAudio_Amplitude
      )
     ))
    ScaleA=true
    ScaleRGB=true
   )
  ]
  rendererClasses [&NiagaraSpriteRendererProperties, &NiagaraLightRendererProperties]
  renderers [
   {rendererClass=&NiagaraSpriteRendererProperties}
   {rendererClass=&NiagaraLightRendererProperties, rendererIndex=1}
  ]
 }
}
summary {
 systemName=NS_Backpack_SkullPunk
 emitters [
  {emitterName=Box_Top, rendererClasses [&NiagaraMeshRendererProperties]}
  {emitterName=Box_Bottom, rendererClasses [&NiagaraMeshRendererProperties]}
  {emitterName=Glitch_Top, rendererClasses [&NiagaraSpriteRendererProperties]}
  {emitterName=Light, rendererClasses [&NiagaraSpriteRendererProperties, &NiagaraLightRendererProperties]}
  {emitterName=Glow, rendererClasses [&NiagaraSpriteRendererProperties]}
  {emitterName=Glitch_Bottom, rendererClasses [&NiagaraSpriteRendererProperties]}
 ]
 userVariables [
  User.bIsFrontEnd {type=NiagaraBool}
  User.bIsFrontEndPreview {type=NiagaraBool}
  User.Offset Back {type=NiagaraFloat, defaultValue=-2.5, description=""}
 ]
}
